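# Material function that samples the Apollo terrain biome texture to output grass colors and biome masks for landscape materials.
# A static switch picks landscape layer coordinates or world-position terrain coordinates, then halves them to address one quadrant of the BiomeTexture parameter, a four-quadrant atlas.
# Offset samples yield grass colors, Mountain, Tropical and Desert masks with an unused alpha, and a CustomBiomeA mask.
# The upper-left quadrant sample feeds no output.
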MaterialFunction /Game/Athena/Apollo/Environments/Landscape/Materials/Biome/CommonMaterials/MF_ApolloTerrainMasks
graph {nodes [
  n1 = TextureSample(UVs=Add(A=NamedRerouteUsage(), B=Constant2Vector() {g=0.5}), Tex=NamedRerouteUsage()) {
   desc=BiomeColors
   samplerSource=SSM_Wrap_WorldGroupSettings
   samplerType=SAMPLERTYPE_LinearColor
   texture=&/T_ApolloBiome2
  }
  FunctionOutput(None=@n1.RGB) {outputName=BiomeGrassColors, sortPriority=-1}
  n2 = TextureSample(UVs=Add(A=NamedRerouteUsage(), B=Constant2Vector() {r=0.5}), Tex=NamedRerouteUsage()) {
   desc=BiomeMasks
   samplerSource=SSM_Wrap_WorldGroupSettings
   samplerType=SAMPLERTYPE_LinearColor
   texture=&T_ApolloBiome2
  }
  FunctionOutput(None=@n2.RGBA) {outputName=BiomeMaskRGBA, sortPriority=-6}
  FunctionOutput(None=@n2.R) {outputName=Mountain, sortPriority=-5}
  FunctionOutput(None=@n2.G) {outputName=Tropical, sortPriority=-4}
  FunctionOutput(None=@n2.B) {outputName=Desert, sortPriority=-3}
  FunctionOutput(None=@n2.A) {outputName=NotInUse, sortPriority=-2}
  n3 = TextureSample(UVs=Add(A=NamedRerouteUsage(), B=Constant2Vector() {g=0.5, r=0.5}), Tex=NamedRerouteUsage()) {
   desc=CustomBiomeMasks - Bottom Right
   samplerSource=SSM_Wrap_WorldGroupSettings
   samplerType=SAMPLERTYPE_LinearColor
   texture=&T_ApolloBiome2
  }
  FunctionOutput(None=@n3.R) {outputName=CustomBiomeA}
  n4 = LandscapeLayerCoords() {mappingScale=2033}
  n5 = MaterialFunctionCall!(WorldPosition) {materialFunction=/Game/Athena/Environments/Landscape/MaterialFunctions/Utilities/MF_Apollo_Terrain_Coords}
  n6 = FunctionInput(Preview=StaticBool()) {bUsePreviewValueAsDefault=true, inputName=UseLandscapeCoords, inputType=FunctionInput_StaticBool}
  NamedRerouteDeclaration(None=Multiply(A=StaticSwitch(True=@n4, False=@n5.Adjustable, Value=@n6), B=Constant() {r=0.5})) {name=BaseCoord, nodeColor=<rgba 1 0 1 1>}
  n7 = TextureObjectParameter() {
   group=BiomeSettings
   parameterName=BiomeTexture
   samplerType=SAMPLERTYPE_LinearColor
   texture=&T_ApolloBiome2
  }
  NamedRerouteDeclaration(None=@n7) {name=BiomeTexture, nodeColor=<rgba 1 0 1 1>}
  TextureSample(UVs=NamedRerouteUsage(), Tex=NamedRerouteUsage()) {
   desc=UpperLeft
   samplerSource=SSM_Wrap_WorldGroupSettings
   samplerType=SAMPLERTYPE_LinearColor
   texture=&T_ApolloBiome2
  }
 ]}
