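# Material function that applies the Frontier custom biome's color adjustments to Apollo landscape terrain layers.
# It takes Grass, Gravel, Mud and Sand color values and lerps each toward a Frontier value using the CustomBiomeA mask from MF_ApolloTerrainMasks.
# Grass, Gravel and Mud targets are hardcoded constants, while Sand uses the Frontier-SandBiomeColorDiv parameter.
# Frontier parameters for GravelBiomeColorDiv, MudBiomeColorDiv and GrassEdge exist but are unconnected.
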MaterialFunction /Game/Athena/Apollo/Environments/Landscape/Materials/Biome/UniqueBiomeMaterialFunctions/MF_CustomBiome_Frontier
graph {nodes [
  Constant3Vector() {constant=<rgba 1 1 1 1>}
  n1 = MaterialFunctionCall!(UseLandscapeCoords) {materialFunction=/Game/Athena/Apollo/Environments/Landscape/Materials/Biome/CommonMaterials/MF_ApolloTerrainMasks}
  FunctionOutput(None=LinearInterpolate(
    A=FunctionInput() {bUsePreviewValueAsDefault=true, inputName=Grass, sortPriority=-12, previewValue=<1 1 1 1>}
    B=Constant3Vector() {constant=<rgba 1.5 1.54 10 1>}
    Alpha=@n1.CustomBiomeA
   )) {outputName=Grass, sortPriority=-12}
  n2 = FunctionInput() {bUsePreviewValueAsDefault=true, inputName=GravelBiomeColorDiv, sortPriority=7, previewValue=<1 1 1 1>}
  FunctionOutput(None=LinearInterpolate(A=@n2, B=Constant3Vector() {constant=<rgba 1.4 1.35 1 1>}, Alpha=@n1.CustomBiomeA)) {outputName=GravelBiomeColorDiv, sortPriority=7}
  n3 = FunctionInput() {bUsePreviewValueAsDefault=true, inputName=MudBiomeColorDiv, sortPriority=9, previewValue=<1 1 1 1>}
  FunctionOutput(None=LinearInterpolate(A=@n3, B=Constant3Vector() {constant=<rgba 1.5 1.3 1 1>}, Alpha=@n1.CustomBiomeA)) {outputName=MudBiomeColorDiv, sortPriority=9}
  n4 = FunctionInput() {bUsePreviewValueAsDefault=true, inputName=SandBiomeColorDiv, sortPriority=12, previewValue=<1 1 1 1>}
  FunctionOutput(None=LinearInterpolate(A=@n4, B=VectorParameter() {parameterName=Frontier-SandBiomeColorDiv, defaultValue=<rgba 1 1 1 1>}.RGB, Alpha=@n1.CustomBiomeA)) {outputName=SandBiomeColorDiv, sortPriority=12}
  VectorParameter() {parameterName=Frontier-GravelBiomeColorDiv, defaultValue=<rgba 1 1 1 1>}
  VectorParameter() {parameterName=Frontier-MudBiomeColorDiv, defaultValue=<rgba 1 1 1 1>}
  VectorParameter() {parameterName=Frontier-GrassEdge, defaultValue=<rgba 1 1 1 1>}
 ]}
