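# Material function that applies custom biome overrides to the Grass, Gravel, Mud and Sand inputs of Apollo landscape materials.
# It chains MF_CustomBiome_Frontier into MF_CustomBiome_Autumn, with static switches that either take each stage's result or pass the four values through.
# Hardcoded static bools enable Frontier and disable Autumn, so the outputs currently carry only the Frontier result.
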MaterialFunction /Game/Athena/Apollo/Environments/Landscape/Materials/Biome/CommonMaterials/MF_CustomBiomes
graph {nodes [
  n1 = FunctionInput() {bUsePreviewValueAsDefault=true, inputName=Grass, sortPriority=-12, previewValue=<1 1 1 1>}
  n2 = FunctionInput() {bUsePreviewValueAsDefault=true, inputName=Gravel, sortPriority=7, previewValue=<1 1 1 1>}
  n3 = FunctionInput() {bUsePreviewValueAsDefault=true, inputName=Mud, sortPriority=9, previewValue=<1 1 1 1>}
  n4 = FunctionInput() {bUsePreviewValueAsDefault=true, inputName=Sand, sortPriority=12, previewValue=<1 1 1 1>}
  n5 = MaterialFunctionCall!(Grass=@n1, GravelBiomeColorDiv=@n2, MudBiomeColorDiv=@n3, SandBiomeColorDiv=@n4) {materialFunction=/Game/Athena/Apollo/Environments/Landscape/Materials/Biome/UniqueBiomeMaterialFunctions/MF_CustomBiome_Frontier}
  n6 = StaticBool() {value=true}
  n7 = StaticSwitch(True=@n5.Grass, False=@n1, Value=@n6)
  n8 = StaticSwitch(True=@n5.GravelBiomeColorDiv, False=@n2, Value=@n6)
  n9 = StaticSwitch(True=@n5.MudBiomeColorDiv, False=@n3, Value=@n6)
  n10 = StaticSwitch(True=@n5.SandBiomeColorDiv, False=@n4, Value=@n6)
  n11 = MaterialFunctionCall!(Grass=@n7, GravelBiomeColorDiv=@n8, MudBiomeColorDiv=@n9, SandBiomeColorDiv=@n10) {materialFunction=/Game/Athena/Apollo/Environments/Landscape/Materials/Biome/UniqueBiomeMaterialFunctions/MF_CustomBiome_Autumn}
  n12 = StaticBool()
  FunctionOutput(None=StaticSwitch(True=@n11.Grass, False=@n7, Value=@n12)) {outputName=Grass, sortPriority=-12}
  FunctionOutput(None=StaticSwitch(True=@n11.GravelBiomeColorDiv, False=@n8, Value=@n12)) {outputName=Gravel, sortPriority=7}
  FunctionOutput(None=StaticSwitch(True=@n11.MudBiomeColorDiv, False=@n9, Value=@n12)) {outputName=Mud, sortPriority=9}
  FunctionOutput(None=StaticSwitch(True=@n11.SandBiomeColorDiv, False=@n10, Value=@n12)) {outputName=Sand, sortPriority=12}
 ]}
