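# Material function that color-corrects Apollo landscape Gravel, Mud, Sand and Grass diffuse inputs per biome and outputs a Grass color.
# Constant colors blend by biome masks, with MF_ApolloTerrainMasks supplying Mountain, Tropical and Desert, then pass through MF_CustomBiomes.
# Diffuse inputs are divided by these colors and saturated, with grass edge floored first.
# Static switches select biome results or defaults, including a UEFN bypass and a default grass color for simple landscapes.
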MaterialFunction /Game/Athena/Apollo/Environments/Landscape/Materials/Biome/CommonMaterials/MF_BiomeColors
graph {nodes [
  Constant3Vector() {desc=Base, constant=<rgba 0.0932 0.15 0.03 1>}
  n1 = FunctionInput() {bUsePreviewValueAsDefault=true, desc=Diffuse input, inputName=Gravel, sortPriority=-9, previewValue=<1 1 1 1>}
  n2 = LinearInterpolate(A=Constant() {r=1}, B=Constant3Vector() {constant=<rgba 2.3 1.6585 1.1855 1>}, Alpha=NamedRerouteUsage())
  n3 = Constant3Vector() {constant=<rgba 1.0192 1.2729 1.3 1>}
  n4 = NamedRerouteUsage()
  n5 = LinearInterpolate(
   A=NamedRerouteDeclaration(None=Constant3Vector() {desc=basic grass, constant=<rgba 0.696 1.1457 5.8 1>}) {name=Common-GrassEdge, nodeColor=<rgba 0.3828 0 1 1>}
   B=Constant3Vector() {desc=mountains, constant=<rgba 1.4468 1.158 9 1>}
   Alpha=NamedRerouteUsage()
  )
  n6 = Constant3Vector() {desc=swamp, constant=<rgba 0.8795 0.75 2.5 1>}
  n7 = NamedRerouteUsage()
  n8 = LinearInterpolate(A=Constant() {r=1}, B=Constant3Vector() {constant=<rgba 2.3 1.2548 0.645 1>}, Alpha=NamedRerouteUsage())
  n9 = Constant3Vector() {constant=<rgba 1 0.8963 0.3916 1>}
  n10 = NamedRerouteUsage()
  n11 = LinearInterpolate(A=Constant() {r=1}, B=Constant3Vector() {constant=<rgba 4 2.6421 2.0761 1>}, Alpha=NamedRerouteUsage())
  n12 = Constant3Vector() {constant=<rgba 1 1 1 1>}
  n13 = NamedRerouteUsage()
  n14 = MaterialFunctionCall!(
   Grass=LinearInterpolate(A=LinearInterpolate(A=@n5, B=@n6, Alpha=@n7), B=Constant3Vector() {desc=farm, constant=<rgba 0.6 1.0024 2 1>}, Alpha=NamedRerouteUsage())
   Gravel=LinearInterpolate(A=LinearInterpolate(A=@n2, B=@n3, Alpha=@n4), B=Constant3Vector() {constant=<rgba 0.671 1.2367 1.1 1>}, Alpha=NamedRerouteUsage())
   Mud=LinearInterpolate(A=LinearInterpolate(A=@n8, B=@n9, Alpha=@n10), B=Constant3Vector() {constant=<rgba 0.543 0.804 0.41 1>}, Alpha=NamedRerouteUsage())
   Sand=LinearInterpolate(A=LinearInterpolate(A=@n11, B=@n12, Alpha=@n13), B=Constant3Vector() {constant=<rgba 1.64 1.886 2.05 1>}, Alpha=NamedRerouteUsage())
  ) {materialFunction=&/MF_CustomBiomes}
  n15 = Saturate(None=Divide(A=@n1, B=@n14.Gravel))
  n16 = NamedRerouteUsage()
  FunctionOutput(None=StaticSwitch(True=@n1, False=StaticSwitch(True=@n15, False=@n1, Value=@n16), Value=NamedRerouteUsage())) {desc=This is used in the mobile path, outputName=Gravel, sortPriority=-9}
  n17 = FunctionInput() {bUsePreviewValueAsDefault=true, desc=Diffuse input, inputName=Mud, sortPriority=-8, previewValue=<1 1 1 1>}
  n18 = Saturate(None=Divide(A=@n17, B=@n14.Mud))
  n19 = NamedRerouteUsage()
  FunctionOutput(None=StaticSwitch(True=@n17, False=StaticSwitch(True=@n18, False=@n17, Value=@n19), Value=NamedRerouteUsage())) {outputName=Mud, sortPriority=-8}
  n20 = FunctionInput() {bUsePreviewValueAsDefault=true, desc=Diffuse input, inputName=Sand, sortPriority=-7, previewValue=<1 1 1 1>}
  n21 = Saturate(None=Divide(A=@n20, B=@n14.Sand))
  n22 = NamedRerouteUsage()
  FunctionOutput(None=StaticSwitch(True=@n20, False=StaticSwitch(True=@n21, False=@n20, Value=@n22), Value=NamedRerouteUsage())) {outputName=Sand, sortPriority=-7}
  n23 = FunctionInput() {bUsePreviewValueAsDefault=true, desc=Diffuse input, inputName=Grass, sortPriority=-10, previewValue=<1 1 1 1>}
  n24 = Max(A=Desaturation(None=@n23, Fraction=Constant()), B=Constant() {r=0.13})
  n25 = Saturate(None=Divide(A=@n24, B=NamedRerouteUsage()) {desc=Default for the base grass edge color only})
  n26 = StaticSwitch(True=Saturate(None=Divide(A=@n24, B=@n14.Grass)), False=@n25, Value=NamedRerouteUsage())
  FunctionOutput(None=StaticSwitch(True=@n25, False=@n26, Value=NamedRerouteUsage())) {outputName=GrassEdge, sortPriority=-10}
  n27 = VectorParameter() {
   desc="Used for simple landscapes like\r\nCreative && Lobby Island"
   group=BiomeSettings
   parameterName=NoBiomes-DefaultGrassColor
   defaultValue=<rgba 0.0932 0.15 0.03 1>
  }
  n28 = StaticSwitch(True=NamedRerouteUsage(), False=@n27.RGB, Value=NamedRerouteUsage())
  FunctionOutput(None=StaticSwitch(True=Constant3Vector() {desc=Non-parameter version, constant=<rgba 0.0932 0.15 0.03 1>}, False=@n28, Value=NamedRerouteUsage())) {outputName=Grass, sortPriority=-11}
  LinearInterpolate(A=@n23, B=@n26, Alpha=ScalarParameter() {defaultValue=1, group=BiomeSettings, parameterName=debug-MFBC-test-edge-overrideA})
  n29 = MaterialFunctionCall!(UseLandscapeCoords) {materialFunction=&/MF_ApolloTerrainMasks}
  NamedRerouteDeclaration(None=@n29.BiomeGrassColors) {name=BiomeGrassColors, nodeColor=<rgba 0.3828 0 1 1>}
  NamedRerouteDeclaration(None=@n29.Mountain) {name=Mountain, nodeColor=<rgba 0.3828 0 1 1>}
  NamedRerouteDeclaration(None=@n29.Tropical) {name=Tropical, nodeColor=<rgba 0.3828 0 1 1>}
  NamedRerouteDeclaration(None=@n29.Desert) {name=Desert, nodeColor=<rgba 0.3828 0 1 1>}
  n30 = FunctionInput(Preview=StaticBool()) {
   bUsePreviewValueAsDefault=true
   desc="Default=False. When set to true, this is designed to hide the \"BiomeColors -UseDefault\" bool param in material instances that don't need it - for UEFN"
   description="Default=False. When set to true, this is designed to hide the \"BiomeColors -UseDefault\" bool param in material instances that don't need it - for UEFN"
   inputName=BypassBiomeColors
   inputType=FunctionInput_StaticBool
   sortPriority=12
  }
  NamedRerouteDeclaration(None=@n30) {name=BypassBiomeColors_StaticBool, nodeColor=<rgba 0.3828 0 1 1>}
  n31 = StaticBoolParameter() {
   defaultValue=true
   desc="True=Uses the BiomeTexture.\r\nFalse=Uses the base material properties"
   group=BiomeSettings
   parameterName=BiomeColors-UseDefault
  }
  NamedRerouteDeclaration(None=@n31) {name=UseBiomeColors_ParameterBool, nodeColor=<rgba 0.3828 0 1 1>}
  Reroute(None=LinearInterpolate(
    A=VectorParameter() {group=BiomeSettings, parameterName=TestNewGrassColor, defaultValue=<rgba 0.0932 0.15 0.03 1>}.RGB
    B=@n28
    Alpha=ScalarParameter() {group=BiomeSettings, parameterName=debug-MFBC-testgg}
   )) {desc="Hook this into \"Output Grass\" for testing"}
  n32 = FunctionInput(Preview=StaticBool()) {
   bUsePreviewValueAsDefault=true
   desc="Try leaving this unhooked for now - not sure if anything needs this && also it might be too easy to break stuff with it"
   inputName=UseLandscapeCoords
   inputType=FunctionInput_StaticBool
   sortPriority=11
  }
  Reroute(None=@n32)
  VectorParameter() {parameterName=Mountain-Gravel, defaultValue=<rgba 2.3 1.6585 1.1855 1>}
  VectorParameter() {parameterName=Tropical-Gravel, defaultValue=<rgba 1.1 1.039 0.7009 1>}
  VectorParameter() {desc=mountains, parameterName=Mountain-GrassEdge, defaultValue=<rgba 1.4468 1.158 9 1>}
  VectorParameter() {desc=swamp, parameterName=Tropical-GrassEdge, defaultValue=<rgba 0.8795 0.75 2.5 1>}
  VectorParameter() {desc=farm, parameterName=Desert-GrassEdge, defaultValue=<rgba 0.6 1.0024 2 1>}
  VectorParameter() {parameterName=Desert-Gravel, defaultValue=<rgba 0.671 1.2367 1.1 1>}
  VectorParameter() {desc=basic grass, parameterName=Common-GrassEdge, defaultValue=<rgba 0.696 1.1457 5.8 1>}
  VectorParameter() {parameterName=Mountain-Mud, defaultValue=<rgba 2.3 1.2548 0.645 1>}
  VectorParameter() {parameterName=Tropical-Mud, defaultValue=<rgba 1 0.8963 0.3916 1>}
  VectorParameter() {parameterName=Desert-Mud, defaultValue=<rgba 0.543 0.804 0.41 1>}
  VectorParameter() {parameterName=Mountain-Sand, defaultValue=<rgba 4 2.6421 2.0761 1>}
  VectorParameter() {parameterName=Tropical-Sand, defaultValue=<rgba 1 1 1 1>}
  VectorParameter() {parameterName=Desert-Sand, defaultValue=<rgba 1.64 1.886 2.05 1>}
 ]}
