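# Material function that height-blends two landscape layer Material Attributes inputs using their grayscale heightmaps and a transition phase.
# Offset heights are subtracted and saturated into an alpha, optionally sharpened by CheapContrast, which blends from Material 2 toward Material 1 while vertex attributes always come from Material 2.
# It also outputs the alpha, the interpolated height and a layer mask taking the maximum of phase and AddLayerMask.
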MaterialFunction /Game/Athena/Apollo/Environments/Landscape/Materials/Biome/CommonMaterials/MF_BlendLandscapeLayers
graph {nodes [
  n1 = FunctionInput() {
   bUsePreviewValueAsDefault=true
   description=Enter a value between 0-1.
   inputName=Transistion Phase
   inputType=FunctionInput_Scalar
   sortPriority=5
  }
  FunctionOutput(None=Max(A=@n1, B=FunctionInput() {inputName=AddLayerMask, inputType=FunctionInput_Scalar, sortPriority=8})) {outputName=LayerMask, sortPriority=3}
  n2 = FunctionInput() {
   bUsePreviewValueAsDefault=true
   description=Enter a grayscale heightmap with values between 0-1.
   inputName=Height Texture 1
   inputType=FunctionInput_Scalar
   sortPriority=-4
  }
  n3 = Add(A=Add(A=@n2, B=OneMinus(None=@n1)))
  n4 = FunctionInput() {
   bUsePreviewValueAsDefault=true
   description=Enter a grayscale heightmap with values between 0-1.
   inputName=Height Texture 2
   inputType=FunctionInput_Scalar
   sortPriority=4
  }
  n5 = Add(A=Multiply(B=@n1) {constA=2}, B=@n4)
  n6 = Saturate(None=ComponentMask(None=Subtract(A=@n3, B=@n5)) {r=true})
  n7 = FunctionInput() {
   bUsePreviewValueAsDefault=true
   description=Uses cheap contrast function and adds one instruction.
   inputName=Contrast
   inputType=FunctionInput_Scalar
   sortPriority=7
  }
  n8 = FunctionInput(Preview=StaticBool()) {bUsePreviewValueAsDefault=true, description="Default=False", inputName=UseContrast, inputType=FunctionInput_StaticBool, sortPriority=6}
  n9 = StaticSwitch(
   True=MaterialFunctionCall(In=@n6, Contrast=@n7) {materialFunction=&CheapContrast}.Result
   False=@n6
   Value=@n8
  )
  FunctionOutput(None=BlendMaterialAttributes(
    A=FunctionInput() {inputName=Material 2, inputType=FunctionInput_MaterialAttributes, sortPriority=-2}
    B=FunctionInput() {inputName=Material 1, inputType=FunctionInput_MaterialAttributes, sortPriority=-5}
    Alpha=FunctionOutput(None=@n9) {description=Value used in lerp function. Positive 1 lerps to texture 1, outputName=Alpha, sortPriority=2}
   ) {desc="Vertex blend type altered!", vertexAttributeBlendType=UseA}) {outputName=BlendedMaterial}
  FunctionOutput(None=LinearInterpolate(A=@n4, B=@n2, Alpha=@n9) {desc=lerp 0 texture 2}) {description="Value used in lerp function. ", outputName=Resulting Height Map, sortPriority=1}
 ]}
