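# Material function that centralizes gravel and mud color modifiers for Apollo landscape biome materials.
# It takes no inputs and exposes four constant color outputs: gravel, gravel edge, an SM gravel edge variant, and mud.
# Output descriptions name the consumers: common and mobile gravel functions, the vertex color layer blend, and mud and dirt-mud modifier functions.
# Gravel edge and mud currently share the same tint.
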MaterialFunction /Game/Athena/Apollo/Environments/Landscape/Materials/Biome/CommonMaterials/MF_GravelAndMudColors
graph {nodes [
  FunctionOutput(None=Constant3Vector() {constant=<rgba 0.9 1 1 0>}) {
   desc="Used in MF_Common_Gravel && MF_Mobile_Common_Gravel"
   description="Used in MF_Common_Gravel && MF_Mobile_Common_Gravel"
   outputName=GravelColorMod
  }
  FunctionOutput(None=Constant3Vector() {constant=<rgba 0.45 0.3773 0.225 1>}) {
   desc="Used in MF_Common_Gravel && MF_Mobile_Common_Gravel"
   description="Used in MF_Common_Gravel && MF_Mobile_Common_Gravel"
   outputName=GravelEdgeColorMod
   sortPriority=10
  }
  FunctionOutput(None=Constant3Vector() {constant=<rgba 0.65 0.608 0.52 1>}) {desc=Used in MF_VextexColorLayerBlend, description=Used in MF_VextexColorLayerBlend, outputName=SM-GravelEdgeColorMod, sortPriority=20}
  FunctionOutput(None=Constant3Vector() {constant=<rgba 0.45 0.3773 0.225 1>}) {desc="Used in MF_Common_Mud && MF_DirtMudModifier", description=Used in MF_Common_Mud, outputName=MudColorMod, sortPriority=30}
 ]}
