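# Landscape material function for the Apollo gravel layer, outputting material attributes, height, displacement and component channels.
# Layer weight and a variance mask blend gravel over dirt textures, tinted by MF_GravelAndMudColors and passed through MF_BiomeColors, with a static bypass input for UEFN.
# Quality switches give Epic a macro-masked height-lerp with dirt-height textures and extra blended normals.
# Several scalar parameters sit unconnected, with matching constants wired instead.
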MaterialFunction /Game/Athena/Apollo/Environments/Landscape/Materials/Biome/CommonMaterials/MF_Common_Gravel
graph {nodes [
  n1 = NamedRerouteUsage()
  n2 = NamedRerouteUsage()
  n3 = FunctionInput(Preview=StaticBool()) {
   bUsePreviewValueAsDefault=true
   description="Default=False. When set to true, this is designed to hide the \"BiomeColors -UseDefault\" bool param in material instances that don't need it - for UEFN"
   inputName=BypassBiomeColors
   inputType=FunctionInput_StaticBool
   sortPriority=100
  }
  n4 = NamedRerouteUsage()
  n5 = SetMaterialAttributes!(
   MaterialAttributes
   BaseColor=MaterialFunctionCall!(Grass, Gravel=QualitySwitch(Default=@n2, Low=@n2, High=@n2, Medium=@n2, Epic=@n1), Mud, Sand, UseLandscapeCoords, BypassBiomeColors=@n3) {materialFunction=&/MF_BiomeColors}.Gravel
   Specular=NamedRerouteUsage()
   Roughness=NamedRerouteUsage()
   Normal=QualitySwitch(Default=@n4, Low=@n4, High=@n4, Medium=@n4, Epic=NamedRerouteUsage())
  ) {attributeSetTypes [
    #1
    #2
    #3
    #4
   ]}
  FunctionOutput(None=@n5) {outputName=Material}
  n6 = LandscapeLayerCoords() {mappingScale=4.5}
  n7 = TextureSample(UVs=@n6) {samplerSource=SSM_Wrap_WorldGroupSettings, texture=/Game/Environments/Apollo/Sets/Terrain/Textures/T_Apollo_Dirt_D_01}
  FunctionOutput(None=@n7.A) {outputName=Height, sortPriority=1}
  FunctionOutput(None=@n7.RGB) {outputName=Diffuse, sortPriority=1}
  n8 = TextureSample(UVs=@n6) {
   samplerSource=SSM_Wrap_WorldGroupSettings
   samplerType=SAMPLERTYPE_Normal
   texture=/Game/Environments/Apollo/Sets/Terrain/Textures/T_Apollo_Dirt_N_01
  }
  n9 = Multiply(A=@n8.RGB, B=Constant3Vector() {constant=<rgba 1.5 1.5 1 1>})
  FunctionOutput(None=@n9) {outputName=Normals, sortPriority=1}
  n10 = NamedRerouteUsage()
  FunctionOutput(None=QualitySwitch(Default=@n10, Low=@n10, High=@n10, Medium=@n10, Epic=NamedRerouteUsage())) {outputName=Height-W/Variance, sortPriority=1}
  n11 = Multiply(A=LandscapeLayerCoords() {mappingScale=1}, B=Constant() {r=0.1})
  n12 = TextureSample(UVs=@n11) {samplerSource=SSM_Wrap_WorldGroupSettings, texture=/Game/Environments/Apollo/Sets/Terrain/Textures/T_Apollo_DirtHeight_D}
  n13 = Multiply(A=LandscapeLayerCoords() {mappingScale=1}, B=Constant() {r=0.07})
  n14 = TextureSample(UVs=@n13) {
   samplerSource=SSM_Wrap_WorldGroupSettings
   samplerType=SAMPLERTYPE_LinearColor
   texture=/Game/Environments/Apollo/Sets/Terrain/Textures/T_Apollo_DirtHeight_M
  }
  n15 = Power(Base=@n14.R, Exp=Constant() {r=1.5})
  n16 = Constant() {r=0.75}
  n17 = MaterialFunctionCall!(
   Diffuse 1=NamedRerouteUsage()
   Diffuse 2=Multiply(A=Constant() {r=0.85}, B=@n12.RGB)
   Transistion Phase=Saturate(None=Multiply(A=@n15, B=@n16) {desc="Can use an additional mask to affect the multiply amount for variation!"})
   Height Texture 1=NamedRerouteUsage()
   Height Texture 2=@n12.A
   Contrast
  ) {materialFunction=/Engine/Functions/Engine_MaterialFunctions02/Texturing/HeightLerpWithTwoHeightMaps}
  n18 = NamedRerouteDeclaration(None=@n17.Resulting Height Map) {name=Epic-Height-W/Variance_1, nodeColor=<rgba 0.5 0 1 1>}
  n19 = Multiply(A=Subtract(A=@n18, B=Constant() {r=0.5}), B=Constant() {r=30})
  n20 = Constant()
  FunctionOutput(None=Multiply(A=Add(A=@n19, B=@n20), B=Constant() {r=1})) {outputName=Displacement, sortPriority=2}
  LinearInterpolate(A=NamedRerouteUsage(), B=@n18, Alpha=Constant() {r=1})
  Multiply(A=@n6, B=ScalarParameter() {defaultValue=1, group=Mud, parameterName=MudScale})
  n21 = FunctionInput() {inputName=LayerWeight, inputType=FunctionInput_Scalar, sortPriority=-10}
  n22 = FunctionInput() {bUsePreviewValueAsDefault=true, inputName=VarianceMask, inputType=FunctionInput_Scalar, previewValue=<1 1 1 1>}
  n23 = Constant() {r=1.65}
  n24 = Multiply(A=@n21, B=Saturate(None=Multiply(A=@n22, B=@n23)))
  n25 = LandscapeLayerCoords() {mappingScale=4}
  n26 = TextureSample(UVs=@n25) {
   samplerSource=SSM_Wrap_WorldGroupSettings
   samplerType=SAMPLERTYPE_Normal
   texture=/Game/Environments/Apollo/Sets/Terrain/Textures/T_Apollo_Gravel_N_01
  }
  NamedRerouteDeclaration(None=LinearInterpolate(A=@n9, B=@n26.RGB, Alpha=@n24)) {name=Normals, nodeColor=<rgba 0.5 0 1 1>}
  n27 = MaterialFunctionCall() {materialFunction=&/MF_GravelAndMudColors}
  n28 = Constant4Vector() {constant=<rgba 0.7 0.7 0.7 1>}
  n29 = TextureSample(UVs=@n25) {samplerSource=SSM_Wrap_WorldGroupSettings, texture=/Game/Environments/Apollo/Sets/Terrain/Textures/T_Apollo_Gravel_D_01}
  n30 = LinearInterpolate(A=@n7.RGBA, B=Multiply(A=Max(A=@n28.RGBA, B=@n29.A), B=@n29.RGBA), Alpha=@n24)
  n31 = Multiply(A=@n27.GravelColorMod, B=ComponentMask(None=@n30) {b=true, g=true, r=true})
  NamedRerouteDeclaration(None=LinearInterpolate(A=Multiply(A=@n27.GravelEdgeColorMod, B=@n31), B=@n31, Alpha=@n21)) {name=Diffuse, nodeColor=<rgba 0.5 0 1 1>}
  n32 = TextureSample(UVs=@n11) {
   samplerSource=SSM_Wrap_WorldGroupSettings
   samplerType=SAMPLERTYPE_Normal
   texture=/Game/Environments/Apollo/Sets/Terrain/Textures/T_Apollo_DirtHeight_N
  }
  n33 = LinearInterpolate(A=@n32.RGB, B=NamedRerouteUsage(), Alpha=@n17.Lerp Alpha No Contrast)
  n34 = TextureSample(UVs=@n13) {
   samplerSource=SSM_Wrap_WorldGroupSettings
   samplerType=SAMPLERTYPE_Normal
   texture=/Game/Global/Textures/Default/Tileable_Patterns/T_DirtMask_N
  }
  n35 = Subtract(A=Constant() {r=1}, B=@n17.Lerp Alpha No Contrast)
  n36 = Multiply(A=Multiply(A=Abs(None=@n35), B=@n17.Lerp Alpha No Contrast), B=Constant() {r=1.25})
  NamedRerouteDeclaration(None=LinearInterpolate(
    A=@n33
    B=MaterialFunctionCall(BaseNormal=@n34.RGB, AdditionalNormal=@n33) {materialFunction=&BlendAngleCorrectedNormals}.Result
    Alpha=Saturate(None=@n36)
   )) {name=Epic-Normals, nodeColor=<rgba 0.5 0 1 1>}
  NamedRerouteDeclaration(None=@n17.Results) {name=Epic-Diffuse, nodeColor=<rgba 0.5 0 1 1>}
  n37 = TextureSample(UVs=@n6) {
   samplerSource=SSM_Wrap_WorldGroupSettings
   samplerType=SAMPLERTYPE_Masks
   texture=/Game/Environments/Apollo/Sets/Terrain/Textures/T_Apollo_Dirt_S_01
  }
  NamedRerouteDeclaration(None=FunctionOutput(None=@n37.R) {outputName=Spec, sortPriority=1}) {name=Spec, nodeColor=<rgba 0.5 0 1 1>}
  NamedRerouteDeclaration(None=FunctionOutput(None=@n37.B) {outputName=Rough, sortPriority=1}) {name=Rough, nodeColor=<rgba 0.5 0 1 1>}
  NamedRerouteDeclaration(None=ComponentMask(None=@n30) {a=true}) {name=Height-W/Variance, nodeColor=<rgba 0.5 0 1 1>}
  ScalarParameter() {defaultValue=1.65, parameterName=Gravel-VarMod}
  ScalarParameter() {defaultValue=0.85, group=Rock Small, parameterName=EPIC_GravelA-DiffuseMod}
  ScalarParameter() {defaultValue=0.5, parameterName=HD_Gravel_Sub1}
  ScalarParameter() {defaultValue=30, parameterName=HD_Gravel_Mod1}
  ScalarParameter() {defaultValue=0.1, group=Rock Small, parameterName=EPIC_GravelA-Scale}
  ScalarParameter() {defaultValue=1, parameterName=HD_Gravel_Mod2}
  ScalarParameter() {parameterName=HD_Gravel_Add1}
  ScalarParameter() {defaultValue=1.5, parameterName=EPIC_GravelMacroHeight-Pow}
  ScalarParameter() {defaultValue=0.75, parameterName=EPIC_GravelMacroHeight-Mod}
  ScalarParameter() {defaultValue=0.07, group=Rock Small, parameterName=EPIC_GravelMacro-Scale}
  ScalarParameter() {defaultValue=1, parameterName=HD_Gravel_DebugAlteredHeight}
  ScalarParameter() {defaultValue=1.25, parameterName=EPIC_GravelMacroMask-Mod}
 ]}
