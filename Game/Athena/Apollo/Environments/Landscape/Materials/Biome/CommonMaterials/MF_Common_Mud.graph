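# Material function that builds the shared mud layer for Apollo landscape biomes, outputting material attributes and an inverted, remapped height.
# Diffuse is tinted by gravel-and-mud colors and passed through MF_BiomeColors, with a static bypass switch for UEFN material instances.
# A HeightLerp mask from squared inverted VarianceMask and Height varies color darkening, roughness and normal flattening.
# The BiomeRoughMod multiply is left unconnected.
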MaterialFunction /Game/Athena/Apollo/Environments/Landscape/Materials/Biome/CommonMaterials/MF_Common_Mud
graph {nodes [
  Constant3Vector() {constant=<rgba 0.45 0.3773 0.225 1>}
  n1 = Multiply(
   A=MaterialFunctionCall() {materialFunction=&/MF_GravelAndMudColors}.MudColorMod
   B=FunctionInput() {bUsePreviewValueAsDefault=true, inputName=Diffuse}
  )
  n2 = FunctionInput(Preview=StaticBool()) {
   bUsePreviewValueAsDefault=true
   description="Default=False. When set to true, this is designed to hide the \"BiomeColors-UseDefault\" bool param in material instances that don't need it - for UEFN"
   inputName=BypassBiomeColors
   inputType=FunctionInput_StaticBool
   sortPriority=100
  }
  n3 = Constant() {r=0.7}
  n4 = FunctionInput() {bUsePreviewValueAsDefault=true, inputName=VarianceMask, inputType=FunctionInput_Scalar, previewValue=<1 1 1 1>}
  n5 = OneMinus(None=@n4)
  n6 = FunctionInput() {bUsePreviewValueAsDefault=true, inputName=Height, inputType=FunctionInput_Scalar, previewValue=<1 1 1 1>}
  n7 = MaterialFunctionCall(Transition Phase=Multiply(A=@n5, B=@n5), Height Texture=@n6) {materialFunction=&HeightLerp}
  n8 = Add(A=@n7.Lerp Alpha No Contrast, B=Constant() {r=0.8})
  n9 = FunctionInput() {bUsePreviewValueAsDefault=true, inputName=Rough, inputType=FunctionInput_Scalar}
  n10 = Constant3Vector() {constant=<rgba 0.25 0.25 1 0>}
  n11 = FunctionInput() {bUsePreviewValueAsDefault=true, inputName=Normals, previewValue=<0 0 1 1>}
  n12 = SetMaterialAttributes!(
   MaterialAttributes
   BaseColor=Multiply(
    A=MaterialFunctionCall!(Grass, Gravel, Mud=@n1, Sand, UseLandscapeCoords, BypassBiomeColors=@n2) {materialFunction=&/MF_BiomeColors}.Mud
    B=LinearInterpolate(A=@n3, Alpha=@n7.Lerp Alpha No Contrast)
   )
   Specular=FunctionInput() {bUsePreviewValueAsDefault=true, inputName=Spec, inputType=FunctionInput_Scalar}
   Roughness=Multiply(A=Saturate(None=@n8), B=@n9)
   Normal=LinearInterpolate(A=Multiply(A=@n10, B=@n11), B=@n11, Alpha=@n7.Lerp Alpha No Contrast)
  ) {attributeSetTypes [
    #1
    #2
    #3
    #4
   ]}
  FunctionOutput(None=@n12) {outputName=Material}
  n13 = Divide(A=OneMinus(None=@n6), B=Constant() {r=2})
  n14 = Constant() {r=0.25}
  FunctionOutput(None=Saturate(None=Add(A=@n13, B=@n14) {constB=0.25})) {outputName=Height, sortPriority=1}
  Multiply(
   A=@n9
   B=FunctionInput(Preview=Constant() {r=1}) {bUsePreviewValueAsDefault=true, inputName=BiomeRoughMod, inputType=FunctionInput_Scalar, sortPriority=4}
  )
 ]}
