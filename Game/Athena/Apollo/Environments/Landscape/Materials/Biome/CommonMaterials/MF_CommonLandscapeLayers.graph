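# Landscape material function for Apollo biomes that height-blends the common grass, sand, gravel, mud and asphalt layers into one material and heightmap.
# Painted layer weights drive chained blends, with a static switch placing sand before gravel or after mud.
# Mud reuses gravel's texture samples, and combined layer weights plus slope feed the grass border color mask.
# It also outputs grass color masks, displacement and blend-alpha-derived foliage masks.
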MaterialFunction /Game/Athena/Apollo/Environments/Landscape/Materials/Biome/CommonMaterials/MF_CommonLandscapeLayers
graph {nodes [
  Constant() {r=20}
  Constant() {r=-1}
  n1 = NamedRerouteUsage()
  n2 = NamedRerouteUsage()
  n3 = MaterialFunctionCall!(UseUnderOceanSandMaterial, BypassBiomeColors, UseOceanLevel=NamedRerouteUsage()) {materialFunction=&/MF_Common_Sand}
  n4 = LandscapeLayerSample() {parameterName=Sand}
  n5 = MaterialFunctionCall!(
   Material 1=@n1
   Height Texture 1=@n2
   Material 2=@n3.Material
   Height Texture 2=@n3.Material
   Transistion Phase=@n4
   UseContrast
   Contrast
   AddLayerMask
  ) {materialFunction=&/MF_BlendLandscapeLayers}
  n6 = NamedRerouteUsage()
  n7 = LandscapeLayerSample() {parameterName=Gravel}
  n8 = MaterialFunctionCall!(LayerWeight=@n7, VarianceMask=NamedRerouteUsage(), BypassBiomeColors) {materialFunction=&/MF_Common_Gravel}
  n9 = MaterialFunctionCall(
   Material 1=StaticSwitch(True=@n5.BlendedMaterial, False=@n1, Value=@n6)
   Height Texture 1=StaticSwitch(True=@n5.Resulting Height Map, False=@n2, Value=@n6)
   Material 2=@n8.Material
   Height Texture 2=@n8.Material
   Transistion Phase=@n7
  ) {materialFunction=&MF_BlendLandscapeLayers}
  n10 = MaterialFunctionCall!(
   Diffuse=FunctionOutput(None=@n8.Diffuse) {outputName=GravelDiffuse, sortPriority=3}
   Height=FunctionOutput(None=@n8.Height) {outputName=GravelHeight, sortPriority=2}
   Spec=FunctionOutput(None=@n8.Spec) {outputName=GravelSpec, sortPriority=4}
   Rough=FunctionOutput(None=@n8.Rough) {outputName=GravelRough, sortPriority=5}
   Normals=FunctionOutput(None=@n8.Normals) {outputName=GravelNormals, sortPriority=6}
   VarianceMask=NamedRerouteUsage()
   BiomeRoughMod
   BypassBiomeColors
  ) {materialFunction=&/MF_Common_Mud}
  n11 = MaterialFunctionCall(
   Material 1=@n9.BlendedMaterial
   Height Texture 1=@n9.BlendedMaterial
   Material 2=@n10.Material
   Height Texture 2=@n10.Material
   Transistion Phase=LandscapeLayerSample() {parameterName=Mud}
  ) {materialFunction=&MF_BlendLandscapeLayers}
  n12 = MaterialFunctionCall(UseOceanLevel=NamedRerouteUsage()) {materialFunction=&MF_Common_Sand}
  n13 = MaterialFunctionCall(
   Material 1=@n11.BlendedMaterial
   Height Texture 1=@n11.BlendedMaterial
   Material 2=@n12.Material
   Height Texture 2=@n12.Material
   Transistion Phase=LandscapeLayerSample() {parameterName=Sand}
  ) {materialFunction=&MF_BlendLandscapeLayers}
  n14 = NamedRerouteUsage()
  n15 = MaterialFunctionCall!(TilingSize) {materialFunction=/Game/Athena/Apollo/Environments/Landscape/Materials/MaterialFunctions/MF_ApolloAsphalt}
  n16 = MaterialFunctionCall(
   Material 1=StaticSwitch(True=@n11.BlendedMaterial, False=@n13.BlendedMaterial, Value=@n14)
   Height Texture 1=StaticSwitch(True=@n11.Resulting Height Map, False=@n13.Resulting Height Map, Value=@n14)
   Material 2=@n15.DefaultLandscapeMaterial
   Height Texture 2=@n15.DefaultLandscapeMaterial
   Transistion Phase=LandscapeLayerSample() {parameterName=Asphalt}
  ) {materialFunction=&MF_BlendLandscapeLayers}
  FunctionOutput(None=@n16.BlendedMaterial) {outputName=Material, sortPriority=-1}
  n17 = Multiply(A=MaterialFunctionCall() {materialFunction=&/MF_Common_Landscape_Layers}.Result) {constB=0}
  n18 = Constant()
  FunctionOutput(None=Add(
    A=@n16.Resulting Height Map
    B=ShadingPathSwitch(Default=@n17, Deferred=@n17, Mobile=@n18) {desc=This is required for the target layers to behave correctly in the landscape tool}
   )) {outputName=Heightmap}
  n19 = LandscapeLayerSample() {parameterName=Mud}
  n20 = LandscapeLayerSample() {parameterName=Gravel}
  FunctionOutput(None=LinearInterpolate(
    A=Reroute() {desc="Connect to \"Displacement\" output of MF_Common_Grass"}
    B=Reroute() {desc="Connect to \"Displacement\" output of MF_Common_Gravel"}
    Alpha=Max(A=@n19, B=@n20)
   )) {outputName=Displacement, sortPriority=11}
  n21 = Constant() {r=5}
  n22 = Constant() {r=-1}
  FunctionOutput(None=Saturate(None=LinearInterpolate(A=@n21, B=@n22, Alpha=@n9.Alpha))) {desc=Used in Papaya version, outputName=GravelMaskForGrassFoliage, sortPriority=12}
  n23 = NamedRerouteUsage()
  n24 = NamedRerouteUsage()
  n25 = NamedRerouteUsage()
  n26 = NamedRerouteUsage()
  FunctionOutput(None=Max(A=Max(A=@n23, B=@n24), B=Max(A=@n25, B=@n26))) {outputName=CommonLayerMaskForGrassFoliage, sortPriority=13}
  n27 = Max(
   A=Max(A=LandscapeLayerSample() {parameterName=Mud}, B=LandscapeLayerSample() {parameterName=Gravel})
   B=LandscapeLayerSample() {parameterName=Sand}
  )
  n28 = LandscapeLayerSample() {parameterName=Asphalt}
  n29 = ComponentMask(None=VertexNormalWS()) {b=true}
  n30 = OneMinus(None=Power(Base=Abs(None=@n29), Exp=Constant() {r=3}))
  n31 = Constant() {r=0.3}
  n32 = Max(
   A=Max(A=Max(A=@n27, B=@n28), B=Multiply(A=@n30, B=@n31))
   B=FunctionInput(Preview=Constant()) {bUsePreviewValueAsDefault=true, inputName=AdditionalGrassBorderColorMask, inputType=FunctionInput_Scalar, sortPriority=-8}
  )
  n33 = Constant() {r=2}
  n34 = MaterialFunctionCall!(CombinedLandscapeLayersMask=Saturate(None=Multiply(A=@n32, B=@n33)), BypassUseRVT, BypassBiomeColors) {materialFunction=&/MF_Common_Grass}
  FunctionOutput(None=@n34.GrassColorMask1) {outputName=GrassColorMask1}
  FunctionOutput(None=@n34.GrassColorMask2) {outputName=GrassColorMask2}
  FunctionOutput(None=@n34.GrassColorMask3) {outputName=GrassColorMask3}
  FunctionOutput(None=@n34.Default-LandscapeHeightmap) {outputName=GrassHeightmap, sortPriority=7}
  FunctionOutput(None=@n34.GrassColorMask4) {outputName=GrassNoiseMask, sortPriority=9}
  FunctionOutput(None=@n34.TangentNormals) {outputName=GrassNormal, sortPriority=8}
  FunctionOutput(None=@n34.BasecolorDetail) {outputName=GrassBaseColorDetail, sortPriority=10}
  NamedRerouteDeclaration(None=FunctionInput(Preview=StaticBool()) {bUsePreviewValueAsDefault=true, inputName=PutSandBeforeGravel, inputType=FunctionInput_StaticBool}) {name=PutSandBeforeGravel, nodeColor=<rgba 0.5 0 1 1>}
  NamedRerouteDeclaration(None=@n34.GrassColorMask1) {name=GrassColorMask1, nodeColor=<rgba 0.5 0 1 1>}
  NamedRerouteDeclaration(None=@n34.Default-LandscapeMaterial) {name=GrassMaterial, nodeColor=<rgba 0.5 0 1 1>}
  NamedRerouteDeclaration(None=@n34.Default-LandscapeHeightmap) {name=GrassHeight, nodeColor=<rgba 0.5 0 1 1>}
  n35 = Saturate(None=LinearInterpolate(A=Constant() {r=30}, B=Constant(), Alpha=@n5.Alpha))
  NamedRerouteDeclaration(None=Saturate(None=Add(A=@n4, B=@n35))) {name=GM-Sand, nodeColor=<rgba 0.5 0 1 1>}
  n36 = Constant() {r=5}
  n37 = Constant() {r=-1}
  NamedRerouteDeclaration(None=Saturate(None=LinearInterpolate(A=@n36, B=@n37, Alpha=@n9.Alpha))) {name=GM-Gravel, nodeColor=<rgba 0.5 0 1 1>}
  n38 = Constant() {r=5}
  n39 = Constant() {r=-1}
  NamedRerouteDeclaration(None=Saturate(None=LinearInterpolate(A=@n38, B=@n39, Alpha=@n11.Alpha))) {name=GM-Mud, nodeColor=<rgba 0.5 0 1 1>}
  n40 = Constant() {r=5}
  n41 = Constant() {r=-1}
  NamedRerouteDeclaration(None=Saturate(None=LinearInterpolate(A=@n40, B=@n41, Alpha=@n16.Alpha))) {name=GM-Asphalt, nodeColor=<rgba 0.5 0 1 1>}
  NamedRerouteDeclaration(None=FunctionInput(Preview=StaticBool() {value=true}) {bUsePreviewValueAsDefault=true, inputName=UseOceanLevel, inputType=FunctionInput_StaticBool, sortPriority=100}) {name=UseOceanLevel, nodeColor=<rgba 0.5 0 1 1>}
  ScalarParameter() {parameterName=FeatureZeroOut}
  n42 = GetMaterialAttributes(None=@n16.BlendedMaterial) {attributeGetTypes [#1]}
  SetMaterialAttributes!(
   MaterialAttributes=@n42.MaterialAttributes
   BaseColor=LinearInterpolate(A=@n42.BaseColor, B=@n16.Resulting Height Map, Alpha=ScalarParameter() {parameterName=debug-ts243rwrt})
  ) {attributeSetTypes [#1]}
 ]}
