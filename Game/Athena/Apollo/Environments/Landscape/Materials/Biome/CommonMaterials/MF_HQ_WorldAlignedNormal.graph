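# Material function for landscape biome materials that reorients an input normal for a Z-axis (top-down) world projection, following the surface's vertex normal.
# Scales the normal's XY by Normal Strength, flips Y, and signs X by whether the vertex normal faces up or down.
# Transforms the result through a basis from CreateThirdOrthogonalVector on the vertex normal and world Y.
# A static switch optionally converts the world-space result to tangent space.
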MaterialFunction /Game/Athena/Apollo/Environments/Landscape/Materials/Biome/CommonMaterials/MF_HQ_WorldAlignedNormal
graph {nodes [
  n1 = FunctionInput(Preview=Constant() {r=1}) {
   bUsePreviewValueAsDefault=true
   description="Scalar Input. Default=1"
   inputName=Normal Strength
   inputType=FunctionInput_Scalar
   sortPriority=50
   previewValue=<0 0 1 1>
  }
  Add(A=Constant3Vector() {constant=<rgba 0 0 1 0>}, B=Multiply(A=Constant2Vector() {g=1, r=1}, B=@n1))
  n2 = VertexNormalWS()
  n3 = Constant3Vector() {constant=<rgba 0 0 1 0>}
  n4 = Multiply(
   A=If(A=DotProduct(A=@n2, B=@n3) {desc=generate sign for this projection}, B=Constant(), "A > B"=Constant() {r=1}, "A < B"=Constant() {r=-1})
   B=Constant3Vector() {constant=<rgba 1 0 0 0>}
  )
  n5 = Constant3Vector() {constant=<rgba 0 -1 1 0>}
  n6 = MaterialFunctionCall(X=@n1, Y=@n1, Z=Constant() {r=1}) {materialFunction=&MakeFloat3}
  n7 = FunctionInput() {bUsePreviewValueAsDefault=true, inputName=Normals, sortPriority=-50, previewValue=<0 0 1 1>}
  n8 = MaterialFunctionCall!(Vector1=@n2, Vector2=Constant3Vector() {constant=<rgba 0 1 0 0>}) {materialFunction=/Engine/Functions/Engine_MaterialFunctions02/Math/CreateThirdOrthogonalVector}
  n9 = MaterialFunctionCall(VectorToTransform=Multiply(A=Add(A=@n4, B=@n5), B=Multiply(A=@n6.Result, B=@n7)), BasisX=@n8.Vector3, BasisY=@n8.Vector3, BasisZ=@n8.Vector3) {
   desc=lerps in when the vertex normal faces X or Negative X
   materialFunction=&Transform3x3Matrix
  }
  n10 = FunctionInput(Preview=StaticBool()) {bUsePreviewValueAsDefault=true, description="Default=False", inputName=TransformToTangentSpace, inputType=FunctionInput_StaticBool}
  FunctionOutput(None=StaticSwitch(
    True=Transform(None=@n9.Transformed Vector) {transformSourceType=TRANSFORMSOURCE_World, transformType=TRANSFORM_Tangent}
    False=@n9.Transformed Vector
    Value=@n10
   ))
 ]}
