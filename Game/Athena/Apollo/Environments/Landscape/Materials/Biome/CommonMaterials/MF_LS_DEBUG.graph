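# MF_LS_DEBUG is a landscape material function giving tech artists and level designers debug views, off by default behind a static switch that passes Material through.
# When enabled, it chains parameter-weighted BaseColor overrides: grid, single-color tint, input as color, steep-slope traversal tint and landscape layer highlight.
# The traversal mask thresholds slope from screen-space derivatives of world position; the layer view samples a hand-edited layer name.
# A nested switch bypasses RVT by outputting the non-RVT material.
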MaterialFunction /Game/Athena/Apollo/Environments/Landscape/Materials/Biome/CommonMaterials/MF_LS_DEBUG
graph {
 properties {description="Debug stuff for TA's && LD's. By default, nothing should be turned on in this."}
 nodes [
  n1 = FunctionInput(Preview=SetMaterialAttributes!(MaterialAttributes, BaseColor) {attributeSetTypes [#1]}) {inputName=Non-RVT material, inputType=FunctionInput_MaterialAttributes, sortPriority=800}
  n2 = FunctionInput() {inputName=Material, inputType=FunctionInput_MaterialAttributes}
  n3 = GetMaterialAttributes(None=@n2) {attributeGetTypes [#1]}
  n4 = MaterialFunctionCall() {materialFunction=/Game/Athena/Apollo/Environments/Landscape/Materials/MaterialFunctions/MF_Apollo_LS_GridDebug}
  n5 = LinearInterpolate(A=@n3.BaseColor, B=@n4.RGB, Alpha=@n4.RGB)
  n6 = Saturate(None=ScalarParameter() {group=Debug, parameterName=DEBUG-ShowGrid, sliderMax=1})
  n7 = GetMaterialAttributes(None=SetMaterialAttributes(MaterialAttributes=@n3.MaterialAttributes, BaseColor=LinearInterpolate(A=@n3.BaseColor, B=@n5, Alpha=@n6)) {attributeSetTypes [#1]}) {attributeGetTypes [#1]}
  n8 = Multiply(A=@n7.BaseColor, B=FunctionInput() {bUsePreviewValueAsDefault=true, inputName=DebugMaterialAs1Color, sortPriority=900})
  n9 = ScalarParameter() {group=Debug, parameterName=DEBUG-MaterialAs1Color}
  n10 = GetMaterialAttributes(None=SetMaterialAttributes(MaterialAttributes=@n7.MaterialAttributes, BaseColor=LinearInterpolate(A=@n7.BaseColor, B=@n8, Alpha=@n9)) {attributeSetTypes [#1]}) {attributeGetTypes [#1]}
  n11 = FunctionInput(Preview=Constant3Vector()) {bUsePreviewValueAsDefault=true, inputName=DebugInputAsBasecolor, sortPriority=10, previewValue=<1 1 1 1>}
  n12 = ScalarParameter() {group=Debug, parameterName=DEBUG-InputAsBasecolor}
  n13 = GetMaterialAttributes(None=SetMaterialAttributes(MaterialAttributes=@n10.MaterialAttributes, BaseColor=LinearInterpolate(A=@n10.BaseColor, B=@n11, Alpha=@n12)) {attributeSetTypes [#1]}) {attributeGetTypes [#1]}
  n14 = Multiply(A=@n13.BaseColor, B=VectorParameter() {group=Debug, parameterName=DEBUG-TraversalMaskColor, defaultValue=<rgba 1 0 0 1>}.RGB)
  n15 = WorldPosition()
  n16 = Normalize(VectorInput=CrossProduct(A=DDX(Value=@n15.XYZ), B=DDY(Value=@n15.XYZ)))
  n17 = LinearInterpolate(A=Constant() {r=43}, B=Constant() {r=-16.5}, Alpha=Abs(None=ComponentMask(None=@n16) {b=true}))
  n18 = Constant() {r=1}
  n19 = Multiply(
   A=ScalarParameter() {desc=Set this to 1 to turn it on, group=Debug, parameterName=DEBUG-TraversalMask}
   B=If(A=Saturate(None=@n17), B=Constant() {r=0.75}, "A > B"=@n18, "A == B"=@n18, "A < B"=Constant())
  )
  n20 = GetMaterialAttributes(None=SetMaterialAttributes(MaterialAttributes=@n13.MaterialAttributes, BaseColor=LinearInterpolate(A=@n13.BaseColor, B=@n14, Alpha=@n19)) {attributeSetTypes [#1]}) {attributeGetTypes [#1]}
  n21 = ScalarParameter() {defaultValue=1, group=Debug, parameterName=DEBUG-VisualizeLandscapeLayerMod}
  n22 = VectorParameter() {group=Debug, parameterName=DEBUG-VisualizeLandscapeLayerColor, defaultValue=<rgba 1 0 0 1>}
  n23 = ScalarParameter() {group=Debug, parameterName=DEBUG-VisualizeLandscapeLayer}
  n24 = LandscapeLayerSample() {desc=Replace this with whatever landscape layer you want to preview, parameterName=BiomeD}
  n25 = SetMaterialAttributes(
   MaterialAttributes=@n20.MaterialAttributes
   BaseColor=LinearInterpolate(A=Multiply(A=@n20.BaseColor, B=@n21), B=Multiply(A=@n20.BaseColor, B=@n22.RGB), Alpha=Multiply(A=@n23, B=@n24))
  ) {attributeSetTypes [#1]}
  n26 = StaticBool() {desc="DO NOT submit this set to true!"}
  FunctionOutput(None=StaticSwitch(
    True=StaticSwitch(True=@n1, False=@n25, Value=@n26) {defaultValue=true, desc=Set the static bool to true to bypass the RVT}
    False=@n2
    Value=StaticBool() {desc="DO NOT submit this set to true!"}
   ) {defaultValue=true})
 ]
}
