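# Material function for Apollo biome landscapes that gates a grass input value, with a static bool to switch landscape grass off entirely.
# The scalar input (default 0) is multiplied by a saturated world Z minus LS_Grass_MaskDepth, removing grass below that height.
# The result is multiplied by the landscape visibility mask to respect holes, and the static switch outputs either this or a constant when disabled.
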MaterialFunction /Game/Athena/Apollo/Environments/Landscape/Materials/Biome/CommonMaterials/MF_LS_GrassSwitch
graph {nodes [
  n1 = ComponentMask(None=WorldPosition().XYZ) {b=true}
  n2 = ScalarParameter() {defaultValue=-1200, parameterName=LS_Grass_MaskDepth}
  n3 = Multiply(
   A=FunctionInput(Preview=Constant()) {bUsePreviewValueAsDefault=true, description=Default is 0, inputType=FunctionInput_Scalar}
   B=Saturate(None=Subtract(A=@n1, B=@n2))
  )
  n4 = LandscapeVisibilityMask()
  FunctionOutput(None=StaticSwitch(True=Multiply(A=@n3, B=@n4), False=Constant(), Value=StaticBool() {desc=Set this to false to turn off landscape grass, value=true}))
 ]}
