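# Material function in the Apollo landscape biome common materials that outputs simple top-down planar UVs from XY position.
# The UsePreskinnedWPO static switch picks instance pre-skinned local position or world position, both excluding shader offsets.
# The local path uses a shader stage switch, passing through a vertex interpolator for the pixel shader.
# The chosen XY is divided by the negated absolute Scale input, so the UVs are always flipped regardless of its sign.
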MaterialFunction /Game/Athena/Apollo/Environments/Landscape/Materials/Biome/CommonMaterials/MF_SimpleWorldUV
graph {nodes [
  n1 = LocalPosition() {includedOffsets=ExcludeOffsets, localOrigin=InstancePreSkinning}
  n2 = StaticSwitchParameter(
   True=ShaderStageSwitch(PixelShader=VertexInterpolator(VS=@n1.XYZ).PS, VertexShader=@n1.XYZ)
   False=WorldPosition() {worldPositionShaderOffset=WPT_ExcludeAllShaderOffsets}.XYZ
  ) {parameterName=UsePreskinnedWPO}
  n3 = Abs(None=FunctionInput(Preview=Constant2Vector() {g=2000, r=2000}) {bUsePreviewValueAsDefault=true, inputName=Scale, inputType=FunctionInput_Vector2})
  FunctionOutput(None=Divide(A=ComponentMask(None=@n2) {g=true, r=true}, B=Multiply(A=@n3) {constB=-1}))
 ]}
